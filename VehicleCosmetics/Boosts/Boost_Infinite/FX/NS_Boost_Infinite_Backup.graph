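# Niagara system for the Infinite car boost cosmetic, named as a backup, placed at the car offset plus single boost position.
# The cone emitter merges three spawn groups into one emitter with sprite and mesh renderers, choosing lifetime, color and scale per group.
# A ribbon trail inherits source movement with decreasing strength, shifts color over life, and moves faster backward in frontend preview.
# A third emitter sprays short-lived, bright orange or blue streak sprites backward.
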NiagaraSystem /VehicleCosmetics/Boosts/Boost_Infinite/FX/NS_Boost_Infinite_Backup
dependencies {
 usedDataInterfaces [
  &NiagaraDataInterfaceArrayColor
  &NiagaraDataInterfaceCurve
  &NiagaraDataInterfaceVectorCurve
  &NiagaraDataInterfaceVector2DCurve
  &NiagaraDataInterfaceSkeletalMesh
 ]
 usedDynamicInputs []
 usedModules [
  &self:SystemSpawnScript.SystemScriptSource.SystemScriptGraph.NiagaraNodeAssignment_0.SetVariables_#1
  &SystemState
 ]
}
emitters {
 NE_Infinite_BoostCone {
  emitterUpdateScript [
   &EmitterState(
    Inactive Response=::NewEnumerator1
    Life Cycle Mode=::NewEnumerator1
    Loop Behavior=::NewEnumerator1
    Loop Duration Mode=::NewEnumerator1
    UseLoopDelay=false
   )
   &SpawnRate(SpawnRate=20, Use Spawn Probability=false)
   &SpawnBurst_Instantaneous(Spawn Group=1, Use Spawn Probability=false, UseLoopCountLimit=false)
   &SpawnBurst_Instantaneous(Spawn Group=2, Use Spawn Probability=false, UseLoopCountLimit=false) {module=SpawnBurst_Instantaneous001}
  ]
  particleSpawnScript [
   &VC_PaintedVariantsOverride() {enabled=false}
   &InitializeParticle(
    Color Mode=::NewEnumerator0
    Lifetime=/VehicleCosmetics/FX/Niagara/Modules/MergeEmitters/VC_FloatBySpawnGroup(
     SpawnGroup or ExecIndex:NiagaraInt32=0
     CustomInt:NiagaraInt32=0
     MaximumSpawnGroups:NiagaraInt32=3
     Float SG 0:NiagaraFloat=1
     Float SG 1:NiagaraFloat=$Emitter.CurrentLoopDuration
     Float SG 2:NiagaraFloat=$Emitter.CurrentLoopDuration
     Float SG 3:NiagaraFloat=0
     Float SG 4:NiagaraFloat=0
     Float SG 5:NiagaraFloat=0
    )
    Mesh Scale Mode=::NewEnumerator3
    Mesh Uniform Scale=1.05
    Position=&AddVectorToPosition(Position=$User.DTCarOffset, Vector=Linked() {linkedVariable {name=User.SingleBoostPos, type=NiagaraPosition}})
    Position Mode=::NewEnumerator1
    Position Offset=<-50 0 4>
    Position Offset Coordinate Space=::Local
    Sprite Rotation Mode=::NewEnumerator2
    Sprite Size=&MaskVector2dBySpawnGroup(
     Masked Vector2D=&RandomRangeVector2D(Minimum=<2.5 120>, Maximum=<2.5 160>)
     Passthrough Vector2D=<10 10>
     Spawn Group Mask=0
    )
    Sprite Size Mode=::NewEnumerator4
    UsePositionOffset=true
    Write Material Random=true
   )
   &AddVelocity(Velocity=/VehicleCosmetics/FX/Niagara/Modules/MergeEmitters/VC_VectorBySpawnGroup(
     SpawnGroup or ExecIndex:NiagaraInt32=0
     CustomInt:NiagaraInt32=0
     MaximumSpawnGroups:NiagaraInt32=3
     Vector Group 0:Vector3f=&RandomRangeVector(Minimum=<-20 -20 -20>, Maximum=<20 20 20>)
     Vector Group 1:Vector3f=<0 0 0>
     Vector Group 2:Vector3f=<0 0 0>
     Vector Group 3:Vector3f=<0 0 0>
     Vector Group 4:Vector3f=<0 0 0>
     Vector Group 5:Vector3f=<0 0 0>
    ))
   &Color(Color=/VehicleCosmetics/FX/Niagara/Modules/MergeEmitters/VC_Dynamic_ColorBySpawnGroup(
     Color SG 0:LinearColor=&SelectLinearColorFromArray(
      Array Sampling Mode=::NewEnumerator0
      Direct Array Index=?
      Color Selection Array=DataInterface() {propertyValues {ColorData [<rgba 0 0.0417 1 1>, <rgba 1 0.23 0.1 1>]}}
      Random Seed=0
     )
     MaximumSpawnGroups:NiagaraInt32=3
     Color SG 1:LinearColor=<rgba 1 0.23 0.1 1>
     Color SG 2:LinearColor=<rgba 2 0.4044 0.15 1>
     Color SG 3:LinearColor=<rgba 1 1 1 1>
     Color SG 4:LinearColor=<rgba 1 1 1 1>
    ))
   /SetVariables(
    Particles.Highlight Color:LinearColor=<rgba 10 8.5633 3.3913 1>
    Particles.RenderVisTag:NiagaraInt32=&SpawnGroup()
   ) {isSetParameters=true}
  ]
  particleUpdateScript [
   &ParticleState()
   &ScaleSpriteSize(
    Scale Sprite Size Mode=::NewEnumerator0
    Uniform Scale Factor=&VC_FloatBySpawnGroup(
     SpawnGroup or ExecIndex:NiagaraInt32=0
     CustomInt:NiagaraInt32=0
     MaximumSpawnGroups:NiagaraInt32=3
     Float SG 0:NiagaraFloat=&FloatFromCurve(FloatCurve=DataInterface() {propertyValues {Curve {keys [{interpMode=RCIM_Cubic, tangentMode=RCTM_User, time=0.75, value=0.6}, {interpMode=RCIM_Cubic, tangentMode=RCTM_User, time=1, value=1}]}}})
     Float SG 1:NiagaraFloat=1
     Float SG 2:NiagaraFloat=0
     Float SG 3:NiagaraFloat=0
     Float SG 4:NiagaraFloat=0
     Float SG 5:NiagaraFloat=0
    )
   )
   &ScaleVelocity(Velocity Scale=&VectorFromFloat(Value=&VC_FloatBySpawnGroup(
      SpawnGroup or ExecIndex:NiagaraInt32=0
      CustomInt:NiagaraInt32=0
      MaximumSpawnGroups:NiagaraInt32=3
      Float SG 0:NiagaraFloat=&FloatFromCurve(
       FloatCurve=DataInterface() {propertyValues {Curve {keys [
           {interpMode=RCIM_Cubic, tangentMode=RCTM_User, time=0, value=3}
           {arriveTangent=-0.6481, interpMode=RCIM_Cubic, leaveTangent=-0.6481, tangentMode=RCTM_User, time=0.4, value=1}
           {interpMode=RCIM_Cubic, tangentMode=RCTM_User, time=0.9, value=-3}
          ]}}}
       Scale Curve=2
      )
      Float SG 1:NiagaraFloat=0
      Float SG 2:NiagaraFloat=0
      Float SG 3:NiagaraFloat=0
      Float SG 4:NiagaraFloat=0
      Float SG 5:NiagaraFloat=0
     )))
   &SolveForcesAndVelocity(Clamp Velocity=false, Limit Acceleration=false)
   &ScaleColor(
    Scale Alpha=&VC_FloatBySpawnGroup(
     SpawnGroup or ExecIndex:NiagaraInt32=0
     CustomInt:NiagaraInt32=0
     MaximumSpawnGroups:NiagaraInt32=3
     Float SG 0:NiagaraFloat=&FloatFromCurve(FloatCurve=DataInterface() {propertyValues {Curve {keys [<at 0 1>,<at 1 0>]}}}, Scale Curve=2)
     Float SG 1:NiagaraFloat=1
     Float SG 2:NiagaraFloat=1
     Float SG 3:NiagaraFloat=0
     Float SG 4:NiagaraFloat=0
     Float SG 5:NiagaraFloat=0
    )
    Scale RGB=&VectorFromFloat(Value=&VC_FloatBySpawnGroup(
      SpawnGroup or ExecIndex:NiagaraInt32=0
      CustomInt:NiagaraInt32=0
      MaximumSpawnGroups:NiagaraInt32=3
      Float SG 0:NiagaraFloat=5
      Float SG 1:NiagaraFloat=7000
      Float SG 2:NiagaraFloat=1.5
      Float SG 3:NiagaraFloat=0
      Float SG 4:NiagaraFloat=0
      Float SG 5:NiagaraFloat=0
     ))
    ScaleA=true
    ScaleRGB=true
   )
   &CameraOffset(Camera Offset Amount=5)
  ]
  rendererClasses [&NiagaraSpriteRendererProperties, &NiagaraMeshRendererProperties]
  renderers [
   {rendererClass=&NiagaraSpriteRendererProperties}
   {rendererClass=&NiagaraSpriteRendererProperties, rendererIndex=1}
   {rendererClass=&NiagaraMeshRendererProperties, rendererIndex=2}
   {rendererClass=&NiagaraMeshRendererProperties, rendererIndex=3}
  ]
 }
 NE_Ribbon_Optimized {
  emitterUpdateScript [
   &EmitterState(Life Cycle Mode=::NewEnumerator1, Loop Behavior=::NewEnumerator1, Loop Duration Mode=::NewEnumerator1, UseLoopDelay=false)
   &VC_FilteredSocketCount()
   &VC_RateQualityHandler(High Spawn Rate=60, Low Spawn Rate=60, Medium Spawn Rate=60, Use Custom Rate?=true)
   &VC_SpawnRibbon_v1(SpawnRate=$Emitter.Spawn Rate)
  ]
  particleSpawnScript [
   &VC_PaintedVariantsOverride() {enabled=false}
   &VC_InitializeRibbon_v1(
    Custom Offset=true
    Lifetime=1.15
    Local Vector Offset=&AddVectorToPosition(
     Position=&AddVectorToPosition(Position=$User.DTCarOffset, Vector=Linked() {linkedVariable {name=User.SingleBoostPos, type=NiagaraPosition}})
     Vector=<-50 0 0>
    )
    Offset Amount=1
    Ribbon Width=140
   )
   &AddVelocity(Velocity=&MakeCustomVectorFromBool(Bool=$User.bIsFrontendPreview, True Vector=<-1500 0 0>, False Vector=<-1000 0 0>)) {module=AddVelocity001}
  ]
  particleUpdateScript [
   &ParticleState()
   &ScaleColor(
    Scale Alpha=&FloatFromCurve(FloatCurve=DataInterface() {propertyValues {Curve {
        defaultValue=1
        keys [
         {arriveTangent=8.1619, interpMode=RCIM_Cubic, leaveTangent=8.1619, tangentMode=RCTM_User, time=0, value=0.025}
         {arriveTangent=-0.0103, interpMode=RCIM_Cubic, leaveTangent=-0.0103, tangentMode=RCTM_User, time=0.1408, value=1}
         {arriveTangent=-0.1126, interpMode=RCIM_Cubic, leaveTangent=-0.1126, tangentMode=RCTM_User, time=0.75, value=1}
         {arriveTangent=-1.167, interpMode=RCIM_Cubic, leaveTangent=-1.167, tangentMode=RCTM_User, time=1, value=0}
        ]
       }}})
    Scale RGB=&VectorFromCurve(
     VectorCurve=DataInterface() {propertyValues {XCurve {keys [<at 0.05 1>,<at 0.3 0.06>]}, YCurve {keys [<at 0.05 0.23>,<at 0.3 0.1>]}, ZCurve {keys [<at 0.05 0.1>,<at 0.3 0.55>]}}}
     Scale Curve=&VectorFromFloat(Value=&FloatFromCurve(FloatCurve=DataInterface() {propertyValues {Curve {keys [
           <at 0 1>
           {arriveTangent=0.0605, interpMode=RCIM_Cubic, leaveTangent=0.0605, tangentMode=RCTM_User, time=0.2, value=1}
           {arriveTangent=0.0957, interpMode=RCIM_Cubic, leaveTangent=0.0957, tangentMode=RCTM_User, time=0.3, value=0.85}
           {arriveTangent=-0.5, interpMode=RCIM_Cubic, leaveTangent=-0.5, time=0.6, value=1}
           <at 1 0.5>
          ]}}}))
    )
    ScaleA=true
    ScaleRGB=true
   )
   &InheritSourceMovement(
    Applied Position Scale=&VectorFromFloat(Value=&FloatFromCurve(FloatCurve=DataInterface() {propertyValues {Curve {keys [
          {interpMode=RCIM_Cubic, tangentMode=RCTM_Break, time=0, value=1}
          {arriveTangent=-2.9025, interpMode=RCIM_Cubic, tangentMode=RCTM_Break, time=1, value=0}
         ]}}}))
    ApplyForce=false
    ApplyPosition=true
    ApplyPositionFalloff=false
   )
   &SolveForcesAndVelocity(Clamp Velocity=false, Limit Acceleration=false)
   /SetVariables(
    Particles.R0_RibbonScale:LinearColor=&VC_DynamicScaleColor(
     Color to Scale:LinearColor=$Particles.Color
     RGB Scale:Vector3f=&VectorFromFloat(Value=0.35)
     Alpha Scale:NiagaraFloat=1.5
    )
    Particles.R0_RibbonWidth:NiagaraFloat=&FloatFromCurve(
     FloatCurve=DataInterface() {propertyValues {Curve {
        defaultValue=1
        keys [
         {arriveTangent=0.6251, interpMode=RCIM_Cubic, leaveTangent=0.6251, leaveTangentWeight=1.1373, tangentMode=RCTM_User, time=0, value=0.3}
         {arriveTangent=0.8873, interpMode=RCIM_Cubic, leaveTangent=0.8873, tangentMode=RCTM_User, time=0.9, value=1}
        ]
       }}}
     Scale Curve=320
    )
   ) {isSetParameters=true}
   &ScaleRibbonWidth(Ribbon Width Scale=&FloatFromCurve(FloatCurve=DataInterface() {propertyValues {Curve {
        defaultValue=1
        keys [
         {arriveTangent=0.6251, interpMode=RCIM_Cubic, leaveTangent=0.6251, leaveTangentWeight=1.1373, tangentMode=RCTM_User, time=0, value=0.25}
         {arriveTangent=-0.0175, interpMode=RCIM_Cubic, leaveTangent=-0.0175, tangentMode=RCTM_User, time=0.9, value=1}
        ]
       }}}))
   &VC_UpdateRibbon_v1()
  ]
  rendererClasses [&NiagaraRibbonRendererProperties]
  renderers [
   {rendererClass=&NiagaraRibbonRendererProperties}
   {rendererClass=&NiagaraRibbonRendererProperties, rendererIndex=1}
  ]
 }
 NE_VC_SingleBoost_Sprite_Parent {
  emitterUpdateScript [
   &EmitterState(Life Cycle Mode=::NewEnumerator1, Loop Behavior=::NewEnumerator1, Loop Duration Mode=::NewEnumerator1, UseLoopDelay=false)
   &SpawnRate(SpawnRate=40, Use Spawn Probability=false)
  ]
  particleSpawnScript [
   &VC_PaintedVariantsOverride() {enabled=false}
   &InitializeParticle(
    Color Mode=::NewEnumerator0
    Lifetime Max=0.85
    Lifetime Min=0.35
    Lifetime Mode=::NewEnumerator1
    Position Offset=&AddVectorToPosition(
     Position=&AddVectorToPosition(Position=$User.DTCarOffset, Vector=Linked() {linkedVariable {name=User.SingleBoostPos, type=NiagaraPosition}})
     Vector=<-100 0 0>
    )
    Position Offset Coordinate Space=::Local
    Sprite Size Max=<1.5 90>
    Sprite Size Min=<1.15 40>
    Sprite Size Mode=::NewEnumerator2
    UsePositionOffset=true
    Write Material Random=true
   )
   &Color(Color=&SelectLinearColorFromArray(
     Array Sampling Mode=::NewEnumerator0
     Direct Array Index=?
     Color Selection Array=DataInterface() {propertyValues {ColorData [<rgba 1 0.23 0.1 1>, <rgba 0 0.3 1 1>]}}
     Random Seed=0
    ))
   &ShapeLocation(MaskSpawnGroup=false, Override Seed=false, Sphere Radius=20, UseOwnerScale=true)
   &AddVelocity(Velocity=&RandomRangeVector(Minimum=<-500 0 0>, Maximum=<-250 0 0>))
  ]
  particleUpdateScript [
   &ParticleState()
   &ScaleSpriteSize(
    Non-Uniform Curve Sprite Scale=DataInterface() {propertyValues {
      XCurve {
       defaultValue=0
       keys [
        <at 0 0 RCIM_Cubic>
        {
         arriveTangent=-0.0381
         arriveTangentWeight=0.0667
         interpMode=RCIM_Cubic
         leaveTangent=-0.0381
         leaveTangentWeight=0.1333
         tangentMode=RCTM_User
         time=0.2
         value=1
        }
        {arriveTangent=-0.0087, arriveTangentWeight=0.2, interpMode=RCIM_Cubic, leaveTangent=-0.0087, tangentMode=RCTM_User, time=0.6, value=1}
        <at 1 0 RCIM_Cubic>
       ]
      }
      YCurve {keys [
        <at 0 0 RCIM_Cubic>
        {arriveTangent=-0.0381, interpMode=RCIM_Cubic, leaveTangent=-0.0381, tangentMode=RCTM_User, time=0.2, value=1}
        {arriveTangent=-0.0087, interpMode=RCIM_Cubic, leaveTangent=-0.0087, tangentMode=RCTM_User, time=0.6, value=0.65}
        <at 1 0 RCIM_Cubic>
       ]}
     }}
    Scale Sprite Size Mode=::NewEnumerator3
   )
   &SolveForcesAndVelocity(Clamp Velocity=false, Limit Acceleration=false)
   &ScaleColor(
    Scale Alpha=&FloatFromCurve(FloatCurve=DataInterface() {propertyValues {Curve {keys [
         {interpMode=RCIM_Cubic, tangentMode=RCTM_User, time=0.1, value=0}
         {
          arriveTangent=0.22
          arriveTangentWeight=0.1807
          interpMode=RCIM_Cubic
          leaveTangentWeight=0.2409
          tangentMode=RCTM_Break
          tangentWeightMode=RCTWM_WeightedBoth
          time=0.35
          value=1
         }
         {arriveTangent=-0.3696, tangentMode=RCTM_User, time=1, value=0}
        ]}}})
    Scale RGB=&VectorFromFloat(Value=50)
    ScaleA=true
    ScaleRGB=true
   )
  ]
  rendererClasses [&NiagaraSpriteRendererProperties]
  renderers [{rendererClass=&NiagaraSpriteRendererProperties}]
 }
}
summary {
 systemName=NS_Boost_Infinite_Backup
 emitters [
  {
   emitterName=NE_Infinite_BoostCone
   rendererClasses [&NiagaraSpriteRendererProperties, &NiagaraMeshRendererProperties]
  }
  {emitterName=NE_Ribbon_Optimized, rendererClasses [&NiagaraRibbonRendererProperties]}
  {emitterName=NE_VC_SingleBoost_Sprite_Parent, rendererClasses [&NiagaraSpriteRendererProperties]}
 ]
 userVariables [
  User.bIsFrontendPreview {type=NiagaraBool}
  User.bIsPainted {type=NiagaraBool}
  User.ColorVariantID {type=NiagaraInt32}
  User.DTCarOffset {type=NiagaraPosition, description="", defaultValue {}}
  User.GameModeID {type=NiagaraInt32}
  User.PaintedColor {type=LinearColor, description="", defaultValue {struct=LinearColor, value=<rgba 1 1 1 1>}}
  User.PaintedScalar {type=NiagaraFloat, defaultValue=1, description=""}
  User.PortCount {type=NiagaraInt32, defaultValue=1}
  User.SingleBoostPos {type=NiagaraPosition, description="", defaultValue {}}
 ]
}
