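# Unlit translucent Niagara ribbon material for the energy trail of the Infinite vehicle boost.
# Panning energy, spark-stripe and cloud textures, partly UV-distorted by scrolling curl noise, form a layered pattern that particle color tints through an emissive adjustment function.
# That pattern also builds an alpha mask with a billow-distorted drift texture, fades at both U ends and a low-U bias boosted by MaskSpeed-driven star-beam noise, scaled by particle alpha.
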Material /VehicleCosmetics/Boosts/Boost_Infinite/Materials/M_EnergyTrail
graph {
 properties {
  bAutomaticallySetUsageInEditor=false
  bUsedWithNiagaraRibbons=true
  blendMode=BLEND_Translucent
  shadingModel=MSM_Unlit
  usedShadingModels=Unlit
 }
 nodes [
  MaterialExpressionVertexColor_3 = ParticleColor()
  n1 = Panner(Coordinate=TextureCoordinate()) {speedX=-0.8}
  n2 = Time()
  n3 = Constant() {r=0.2}
  n4 = TextureSample(UVs=Panner(Coordinate=TextureCoordinate(), Time=Multiply(A=@n2, B=@n3)) {speedX=-1, speedY=0.5}) {texture=/DealershipSharedLibrary/FX/Textures/Noise/T_VCVFX_Flow_Curl_01}
  n5 = Add(A=TextureCoordinate() {uTiling=2}, B=Multiply(A=ComponentMask(None=@n4.RGB) {g=true, r=true}, B=Constant() {r=0.05}))
  n6 = Constant() {r=5}
  n7 = Multiply(A=TextureSample(UVs=Panner(Coordinate=@n5) {speedX=-1.5}) {texture=/DealershipSharedLibrary/FX/Textures/Noise/T_VCVFX_Energy_Alpha}.R, B=@n6)
  n8 = Time()
  n9 = Constant() {r=0.2}
  n10 = TextureSample(UVs=Panner(Coordinate=TextureCoordinate(), Time=Multiply(A=@n8, B=@n9)) {speedX=-1, speedY=0.5}) {texture=&T_VCVFX_Flow_Curl_01}
  n11 = Add(A=TextureCoordinate() {uTiling=2}, B=Multiply(A=ComponentMask(None=@n10.RGB) {g=true, r=true}, B=Constant() {r=0.05}))
  n12 = Power(
   Base=TextureSample(UVs=Panner(Coordinate=@n11) {speedX=-1.5}) {texture=&T_VCVFX_Energy_Alpha}.RGB
   Exp=Constant() {r=3}
  )
  n13 = Constant() {r=10}
  n14 = Panner(Coordinate=TextureCoordinate() {uTiling=2}) {speedX=-1}
  n15 = Multiply(
   A=@n6
   B=Multiply(
    A=Multiply(A=@n12, B=@n13)
    B=TextureSample(UVs=@n14) {texture=/DealershipSharedLibrary/FX/Textures/Gradient_Masks/T_VCVFX_Trail_ToonSparks_GradientStripeTyre_A}.RGB
   )
  )
  n16 = TextureCoordinate() {uTiling=2}
  n17 = Add(
   A=Multiply(A=Constant() {r=2}, B=TextureSample(UVs=@n1) {texture=/DealershipSharedLibrary/FX/Textures/Noise/T_VCVFX_CloudDistort_A}.RGB)
   B=Multiply(A=Add(A=@n7, B=@n15), B=TextureSample(UVs=@n16) {texture=/DealershipSharedLibrary/FX/Textures/Gradient_Masks/T_VCVFX_TMaskDrift_A}.R)
  )
  n18 = TextureCoordinate()
  n19 = Multiply(A=Time(), B=Constant() {r=0.2})
  n20 = ComponentMask(None=TextureSample(UVs=Panner(Coordinate=@n18, Time=@n19) {speedX=-1, speedY=0.5}) {texture=/DealershipSharedLibrary/FX/Textures/Noise/T_VCVFX_BillowNoiseTiledOffset01_RGB}.RGB) {g=true, r=true}
  n21 = Constant() {r=0.07}
  n22 = Panner(Coordinate=Add(A=@n16, B=Multiply(A=@n20, B=@n21))) {speedX=-1}
  n23 = ComponentMask(None=TextureCoordinate()) {r=true}
  n24 = Power(Base=@n23, Exp=Constant() {r=2})
  n25 = Power(Base=OneMinus(None=@n23), Exp=Constant() {r=2})
  n26 = Multiply(
   A=Multiply(A=@n17, B=TextureSample(UVs=@n22) {texture=&T_VCVFX_TMaskDrift_A}.R)
   B=Multiply(A=Multiply(A=@n24, B=@n25), B=Constant() {r=15})
  )
  n27 = ComponentMask(None=TextureCoordinate()) {r=true}
  n28 = Multiply(A=Power(Base=OneMinus(None=@n27), Exp=Constant() {r=5}), B=Constant() {r=10})
  n29 = Multiply(A=ScalarParameter() {defaultValue=1, parameterName=MaskSpeed}, B=Time())
  n30 = TextureSample(UVs=Panner(Coordinate=TextureCoordinate(), Time=@n29) {speedX=-0.75, speedY=-0.1}) {texture=/VehicleCosmetics/FX/Textures/Noise/T_VCVFX_NoiseStarBeam_A}
  n31 = Constant() {r=1.5}
  n32 = TextureSample(UVs=Panner(Coordinate=TextureCoordinate() {uTiling=0.5, vTiling=0.5}, Time=@n29) {speedX=-0.75, speedY=-0.1}) {texture=&T_VCVFX_NoiseStarBeam_A}
  MaterialExpressionConstantClamp_3 = Saturate(None=Add(A=Multiply(A=@n30.R, B=@n31), B=Multiply(A=@n31, B=@n32.R)))
  n33 = Add(A=@n28, B=Multiply(A=@n28, B=@MaterialExpressionConstantClamp_3))
  MaterialExpressionConstantClamp_8 = Saturate(None=Multiply(A=@MaterialExpressionVertexColor_3.A, B=Multiply(A=@n26, B=@n33)))
  MaterialFunctionCall(Emissive=LinearInterpolate(B=@n17, Alpha=@MaterialExpressionVertexColor_3.RGB)) {materialFunction=&MF_FX_EmissiveAdjustments}
 ]
}
